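# Unlit Landmass erosion-brush material that computes one water-levelling step of hydraulic erosion on a height render target.
# A Custom HLSL node reads terrain (R) and water (G) from HeightRT, adds rain that dissolves terrain, then levels water across eight neighbours, clamped to available volume.
# Evaporation then redeposits terrain.
# MPC_Erosion supplies simulation settings, and its idx drives a sine-rotated wind vector that shifts the neighbour samples.
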Material /Landmass/Landscape/ErosionBrush/Materials/WaterLevelling/Erosion_WaterLevelling
graph {
 properties {bAllowNegativeEmissiveColor=true, pixelDepthOffsetMode=PDOM_Legacy, shadingModel=MSM_Unlit, usedShadingModels=Unlit}
 nodes [
  MaterialExpressionTextureObjectParameter_0 = TextureObjectParameter() {parameterName=HeightRT, texture=&Black}
  MaterialExpressionMaterialFunctionCall_14 = MaterialFunctionCall() {materialFunction=/Landmass/Landscape/OldPrototype_BP/Mats/MF/PointSampleUVs}
  MaterialExpressionAppendVector_1 = AppendVector(
   A=CollectionParameter() {collection=/Landmass/Landscape/BlueprintBrushes/MPC/MPC_Erosion, parameterName=ResX}
   B=CollectionParameter() {collection=&MPC_Erosion, parameterName=ResY}
  )
  MaterialExpressionCollectionParameter_4 = CollectionParameter() {collection=&MPC_Erosion, parameterName=RainAmount}
  MaterialExpressionCollectionParameter_2 = CollectionParameter() {collection=&MPC_Erosion, parameterName=Solubility}
  MaterialExpressionCollectionParameter_5 = CollectionParameter() {collection=&MPC_Erosion, parameterName=Evaporation}
  MaterialExpressionCollectionParameter_6 = CollectionParameter() {collection=&MPC_Erosion, parameterName=deltaT}
  n1 = CollectionParameter() {collection=&MPC_Erosion, parameterName=idx}
  n2 = Sine(None=Multiply(A=Divide(A=@n1) {constB=1000}, B=ScalarParameter() {defaultValue=0.05, parameterName=Wind Rotation Speed}))
  n3 = ScalarParameter() {defaultValue=0.1, parameterName=Wind Rotation}
  n4 = Multiply(A=VectorParameter() {parameterName=WindDirection, defaultValue=<rgba 2 0.5 0 0>}.RGB, B=ScalarParameter() {parameterName=Wind Intensity})
  MaterialExpressionAdd_1 = Add(
   A=RotateAboutAxis(
    NormalizedRotationAxis=Constant3Vector() {constant=<rgba 0 0 1 0>}
    RotationAngle=Multiply(A=@n2, B=@n3)
    PivotPoint=Constant()
    Position=@n4
   )
   B=@n4
  )
  n5 = Custom!(
   Tex=@MaterialExpressionTextureObjectParameter_0
   UV=@MaterialExpressionMaterialFunctionCall_14.TexelCenterUVs
   TextureSize=@MaterialExpressionAppendVector_1
   RainAmount=@MaterialExpressionCollectionParameter_4
   Solubility=@MaterialExpressionCollectionParameter_2
   Evaporation=@MaterialExpressionCollectionParameter_5
   deltaT=@MaterialExpressionCollectionParameter_6
   Wind=@MaterialExpressionAdd_1
  ) {
   code="//UV input UV\r\n//Tex Input tex height\r\n//TextureSize input texturesize\r\n//RainAmount\r\n//Solubility\r\n//Evaporation\r\n//deltaT \r\n\r\nfloat2 TexelSize = 1 / TextureSize;\r\nfloat2 offsets[8] = {float2(1,0), float2(-1,0), float2(0,1), float2(0,-1),float2(-1,-1), float2(1,-1), float2(-1,1), float2(1,1)};\r\n\r\nfloat4 startsample = Tex.SampleLevel(TexSampler, UV, 0);\r\n\r\n//Perform rain and dissolution\r\nstartsample.r -= RainAmount * Solubility * deltaT;\r\nstartsample.g += RainAmount * deltaT;\r\n\r\nfloat accum = 0;\r\n\r\nfloat startalt = startsample.r + startsample.g;\r\n\r\nfloat altchange = 0;\r\n\r\nfor(int i = 0; i < 8; i++)\r\n{\r\n\t\r\n\tfloat4 offsetsample = Tex.SampleLevel(TexSampler, UV + ((offsets[i] + Wind) * TexelSize), 0);\r\n\t//Perform rain and dissolution\r\n\toffsetsample.r -= RainAmount * Solubility * deltaT;\r\n\toffsetsample.g += RainAmount * deltaT;\r\n\r\n\tfloat offsetalt = offsetsample.r + offsetsample.g;\r\n\tfloat diff = (offsetalt - startalt);\r\n\r\n\tdiff = max(diff, -startsample.g);\r\n\tdiff = min(diff, offsetsample.g);\r\n\r\n\taltchange += diff/8;\r\n\r\n}\r\n\r\nfloat watervol = startsample.g + altchange;\r\n\r\n//perform evaporation and precipitation\r\n\r\nfloat outterrain = startsample.r + (watervol * Evaporation * Solubility * deltaT);\r\nwatervol -= watervol * Evaporation * deltaT;\r\n\r\nreturn float4(outterrain, watervol, watervol, startsample.a);"
   description=Level
   outputType=CMOT_Float4
   inputs [
    {
     inputName=Tex
     input {expression=&self:MaterialExpressionTextureObjectParameter_0}
    }
    {
     inputName=UV
     input {expression=&self:MaterialExpressionMaterialFunctionCall_14}
    }
    {
     inputName=TextureSize
     input {expression=&self:MaterialExpressionAppendVector_1}
    }
    {
     inputName=RainAmount
     input {expression=&self:MaterialExpressionCollectionParameter_4}
    }
    {
     inputName=Solubility
     input {expression=&self:MaterialExpressionCollectionParameter_2}
    }
    {
     inputName=Evaporation
     input {expression=&self:MaterialExpressionCollectionParameter_5}
    }
    {
     inputName=deltaT
     input {expression=&self:MaterialExpressionCollectionParameter_6}
    }
    {inputName=Wind, input {expression=&self:MaterialExpressionAdd_1}}
   ]
  }
  Add(A=ComponentMask(None=@n5) {b=true, g=true, r=true}, B=ScalarParameter() {parameterName=asdfasdfadsf})
  TextureSampleParameter2D() {parameterName=HeightRT, texture=&Black}
 ]
}
