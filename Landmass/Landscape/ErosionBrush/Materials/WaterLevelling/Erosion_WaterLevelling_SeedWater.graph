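# Unlit material that seeds water for the water levelling stage of the Landmass landscape erosion brush.
# It packs the red channel of the HeightRT texture parameter into X and the StartWater scalar from MPC_Erosion into Y through MakeFloat3.
# Negative emissive values are allowed, which suits writing raw height and water data rather than color.
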Material /Landmass/Landscape/ErosionBrush/Materials/WaterLevelling/Erosion_WaterLevelling_SeedWater
graph {
 properties {bAllowNegativeEmissiveColor=true, pixelDepthOffsetMode=PDOM_Legacy, shadingModel=MSM_Unlit, usedShadingModels=Unlit}
 nodes [MaterialFunctionCall(
   X=TextureSampleParameter2D() {parameterName=HeightRT, texture=&Black}.R
   Y=CollectionParameter() {collection=/Landmass/Landscape/BlueprintBrushes/MPC/MPC_Erosion, parameterName=StartWater}
  ) {materialFunction=&MakeFloat3}]
}
