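# Unlit Landmass erosion brush material that computes a weightmap value from the water levelling pass, allowing negative emissive color.
# It divides the HeightRT green channel by the MPC_Erosion GridSize, treating positive values as added soil and negative values as erosion.
# Each part is clamped to a positive magnitude, scaled by the Deposition or Erosion parameter, then summed.
# A separate branch subtracts HeightRTPrev green from HeightRT green.
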Material /Landmass/Landscape/ErosionBrush/Materials/WaterLevelling/Erosion_WaterLevelling_Weightmap
graph {
 properties {bAllowNegativeEmissiveColor=true, pixelDepthOffsetMode=PDOM_Legacy, shadingModel=MSM_Unlit, usedShadingModels=Unlit}
 nodes [
  n1 = TextureSampleParameter2D() {parameterName=HeightRT, texture=&Black}
  n2 = Divide(A=@n1.G, B=CollectionParameter() {collection=/Landmass/Landscape/BlueprintBrushes/MPC/MPC_Erosion, parameterName=GridSize})
  n3 = Multiply(A=@n2) {constB=-1}
  Add(
   A=Multiply(A=Max(A=@n2) {constB=0, desc="Positive = Add soil to terrain"}, B=ScalarParameter() {parameterName=Deposition})
   B=Multiply(A=Max(A=@n3) {constB=0, desc="Negative = Erode"}, B=ScalarParameter() {parameterName=Erosion})
  )
  Subtract(A=@n1.G, B=TextureSampleParameter2D() {parameterName=HeightRTPrev, texture=&Black}.G)
 ]
}
